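# Niagara sprite effect for the RoundThumb emote that flashes a short burst of faint speed lines around the character.
# One Speedline emitter spawns short-lived sprites using skeletal mesh and box location modules, with randomized velocity along simulation-space X.
# Curves stretch each sprite lengthwise while its width and alpha rise quickly then fade, with alpha strength picked from an array.
# Brightness varies randomly, adjusted for front end versus gameplay and time of day.
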NiagaraSystem /BRCosmetics/Animation/Game/MainPlayer/Emotes/RoundThumb/FX/NS_Emote_RoundThumb_Speedlines02
dependencies {
 usedDataInterfaces [
  &NiagaraDataInterfaceCurve
  &NiagaraDataInterfaceVector2DCurve
  &NiagaraDataInterfaceArrayFloat
  &NiagaraDataInterfaceSkeletalMesh
 ]
 usedDynamicInputs [&MaskFloatBySpawnGroup]
 usedModules [
  &self:SystemSpawnScript.SystemScriptSource.SystemScriptGraph.NiagaraNodeAssignment_0.SetVariables_#1
  &SystemState
 ]
}
emitters {Speedline {
  emitterUpdateScript [
   &EmitterState(Life Cycle Mode=::NewEnumerator1, Loop Behavior=::NewEnumerator1, Loop Duration=1, UseLoopDelay=false)
   &SpawnBurst_Instantaneous(
    Spawn Count=&RandomRangeInt(Minimum=14, Maximum=20)
    Spawn Time=0.05
    Use Spawn Probability=false
    UseLoopCountLimit=false
   )
  ]
  particleSpawnScript [
   &InitializeParticle(
    Color=<rgba 0.06 0.0429 0.0469 0.8>
    Lifetime Max=0.3
    Lifetime Min=0.15
    Lifetime Mode=::NewEnumerator1
    Mass Max=2
    Mass Mode=::NewEnumerator3
    Sprite Size Max=<1.2 20>
    Sprite Size Min=<0.7 10>
    Sprite Size Mode=::NewEnumerator2
    UsePositionOffset=false
    Write Material Random=true
   )
   &FN_SetSpawnGroup(Number of Spawn Groups=2)
   &SkeletalMeshLocation(Bone Sampling Mode=::NewEnumerator4, MaskSpawnGroup=false, Skeletal Mesh=$System.Skeletal Mesh)
   &AddVelocity(
    Rotation Coordinate Space=::Simulation
    Velocity=<1 0 0>
    Velocity Speed Scale=&RandomRangeFloat(Minimum=75, Maximum=150)
   )
   &ShapeLocation(
    Box Size=<100 40 45>
    MaskSpawnGroup=false
    Offset=&MakeVector(X=&RandomRangeFloat(Minimum=-40, Maximum=-5), Y=0, Z=0)
    Offset Mode=::NewEnumerator0
    Override Seed=false
    Shape Primitive=::NewEnumerator2
    Surface Only=false
    UseOwnerScale=false
   )
  ]
  particleUpdateScript [
   &ParticleState()
   &ScaleSpriteSize(
    Non-Uniform Curve Scale=&MakeVector2D(X=1, Y=5)
    Non-Uniform Curve Sprite Scale=DataInterface() {propertyValues {
      XCurve {keys [
        {arriveTangent=9.0725, interpMode=RCIM_Cubic, leaveTangent=9.0725, tangentMode=RCTM_User, time=0, value=0}
        {arriveTangent=-2.5249, interpMode=RCIM_Cubic, leaveTangent=-2.5249, tangentMode=RCTM_User, time=0.1408, value=0.9847}
        {arriveTangent=-0.0405, interpMode=RCIM_Cubic, leaveTangent=-0.0405, tangentMode=RCTM_User, time=1, value=0}
       ]}
      YCurve {keys [
        {arriveTangent=4.1785, interpMode=RCIM_Cubic, leaveTangent=4.1785, tangentMode=RCTM_User, time=0, value=0}
        {arriveTangent=1.9119, interpMode=RCIM_Cubic, leaveTangent=1.9119, tangentMode=RCTM_User, time=0.2042, value=0.6228}
        {arriveTangent=0.0224, interpMode=RCIM_Cubic, leaveTangent=0.0224, tangentMode=RCTM_User, time=1, value=1.225}
       ]}
     }}
    Scale Sprite Size Mode=::NewEnumerator3
   )
   &ScaleColor(
    Scale Alpha=&FloatFromCurve(
     FloatCurve=DataInterface() {propertyValues {Curve {keys [
         {arriveTangent=9.0725, interpMode=RCIM_Cubic, leaveTangent=9.0725, tangentMode=RCTM_User, time=0, value=0}
         {arriveTangent=-2.5249, interpMode=RCIM_Cubic, leaveTangent=-2.5249, tangentMode=RCTM_User, time=0.1408, value=0.9847}
         {arriveTangent=-0.0405, interpMode=RCIM_Cubic, leaveTangent=-0.0405, tangentMode=RCTM_User, time=1, value=0}
        ]}}}
     Scale Curve=&SelectFloatFromArray(
      Array Sampling Mode=::NewEnumerator0
      Direct Array Index=?
      Float Selection Array=DataInterface() {propertyValues {FloatData [1,1.5,0.3,0.5]}}
      Random Seed=0
     )
    )
    Scale RGB=&VectorFromFloat(Value=&Multiply_Float(
      A=&FN_FrontEndFloatModifier(
       Gameplay=&RandomRangeFloat(Minimum=0.5, Maximum=1)
       Front End=&RandomRangeFloat(Minimum=0.2, Maximum=0.5)
      )
      B=&FNTimeOfDayScalar(Night Scale=0.2, Morning Scale=0.4, Evening Scale=0.5)
     ))
    ScaleA=true
    ScaleRGB=true
   )
  ]
  rendererClasses [&NiagaraSpriteRendererProperties]
  renderers [{rendererClass=&NiagaraSpriteRendererProperties}]
 }}
summary {
 systemName=NS_Emote_RoundThumb_Speedlines02
 emitters [{emitterName=Speedline, rendererClasses [&NiagaraSpriteRendererProperties]}]
 userVariables [
  User.bIsFrontEnd {type=NiagaraBool}
  User.bIsFrontEndPreview {type=NiagaraBool}
  User.NormalizedNotifyProgress {type=NiagaraFloat, defaultValue=0, description=""}
 ]
}
